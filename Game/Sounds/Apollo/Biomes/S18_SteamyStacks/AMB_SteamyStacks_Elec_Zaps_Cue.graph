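# Sound cue for electrical zap ambience in the Steamy Stacks biome, routed to the Fort_Ambient_Athena sound class.
# An equally weighted Random node picks one of eight zap waves, and a Modulator slightly varies pitch and volume.
# Attenuation and concurrency overrides give it a short-range falloff, listener focus that quiets off-axis zaps, and a small cap on simultaneous plays.
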SoundCue /Game/Sounds/Apollo/Biomes/S18_SteamyStacks/AMB_SteamyStacks_Elec_Zaps_Cue
root=Modulator_0
nodes [
 {id=Modulator_0, children [Random_0], properties {pitchMax=1.1, pitchMin=0.9, volumeMax=1, volumeMin=0.7}}
 {
  id=Random_0
  children [WavePlayer_0, WavePlayer_8, WavePlayer_9, WavePlayer_10, WavePlayer_11, WavePlayer_12, WavePlayer_13, WavePlayer_14]
  properties {weights [1,1,1,1,1,1,1,1]}
 }
 {id=WavePlayer_0, properties {soundWaveAssetPtr=&/AMB_SteamyStacks_Elec_Zaps_01}}
 {id=WavePlayer_10, properties {soundWaveAssetPtr=&/AMB_SteamyStacks_Elec_Zaps_04}}
 {id=WavePlayer_11, properties {soundWaveAssetPtr=&/AMB_SteamyStacks_Elec_Zaps_05}}
 {id=WavePlayer_12, properties {soundWaveAssetPtr=&/AMB_SteamyStacks_Elec_Zaps_06}}
 {id=WavePlayer_13, properties {soundWaveAssetPtr=&/AMB_SteamyStacks_Elec_Zaps_07}}
 {id=WavePlayer_14, properties {soundWaveAssetPtr=&/AMB_SteamyStacks_Elec_Zaps_08}}
 {id=WavePlayer_8, properties {soundWaveAssetPtr=&/AMB_SteamyStacks_Elec_Zaps_02}}
 {id=WavePlayer_9, properties {soundWaveAssetPtr=&/AMB_SteamyStacks_Elec_Zaps_03}}
]
properties {
 bOverrideAttenuation=true
 bOverrideConcurrency=true
 maxDistance=1500
 soundClassObject=&Fort_Ambient_Athena
 volumeMultiplier=0.351
 attenuationOverrides {
  bEnableFocusInterpolation=true
  bEnableListenerFocus=true
  falloffDistance=1500
  focusAttackInterpSpeed=0.7
  focusAzimuth=20
  focusReleaseInterpSpeed=0.7
  nonFocusAzimuth=40
  nonFocusVolumeAttenuation=0.5
  attenuationShapeExtents {x=0}
 }
 concurrencyOverrides {maxCount=3}
}
