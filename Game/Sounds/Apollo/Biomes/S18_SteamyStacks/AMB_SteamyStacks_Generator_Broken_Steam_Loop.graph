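# Looping ambient sound cue for a broken, steaming generator in the Steamy Stacks biome.
# A single looping wave player of the broken-generator steam recording feeds a modulator at the root.
# The cue plays on the ambient sound class and starts from a medium prop emitter attenuation asset, with overrides enabled for its falloff distance and shape extents.
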SoundCue /Game/Sounds/Apollo/Biomes/S18_SteamyStacks/AMB_SteamyStacks_Generator_Broken_Steam_Loop
root=Modulator_1
nodes [
 {id=Modulator_1, children [WavePlayer_0]}
 {
  id=WavePlayer_0
  properties {bLooping=true, soundWaveAssetPtr=&/AMB_Generator_Broken_Steaming_Loop_01}
 }
]
properties {
 attenuationSettings=/Game/Sounds/Attenuation/Fort_Amb_Emitter_Prop_Med_Att
 bOverrideAttenuation=true
 maxDistance=2000
 soundClassObject=&Fort_Ambient_Athena
 volumeMultiplier=1.002
 attenuationOverrides {falloffDistance=2000, attenuationShapeExtents {x=0}}
}
