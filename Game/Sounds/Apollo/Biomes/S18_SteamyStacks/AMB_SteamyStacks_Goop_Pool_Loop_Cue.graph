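# Looping ambient sound cue for the cube goop pool in the Steamy Stacks biome.
# A single looping wave player feeds a modulator at the root, with reduced cue volume on the Athena ambient sound class.
# It references a medium prop emitter attenuation asset but overrides it with non-spatialized falloff and its own occlusion trace and low-pass settings.
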SoundCue /Game/Sounds/Apollo/Biomes/S18_SteamyStacks/AMB_SteamyStacks_Goop_Pool_Loop_Cue
root=Modulator_1
nodes [
 {id=Modulator_1, children [WavePlayer_1]}
 {
  id=WavePlayer_1
  properties {bLooping=true, soundWaveAssetPtr=&/AMB_SteamyStacks_CubeGoop_Pool_Loop_01}
 }
]
properties {
 attenuationSettings=/Game/Sounds/Attenuation/Fort_Amb_Emitter_Prop_Med_Att
 bOverrideAttenuation=true
 maxDistance=2500
 soundClassObject=&Fort_Ambient_Athena
 volumeMultiplier=0.501
 attenuationOverrides {
  bSpatialize=false
  falloffDistance=2500
  occlusionLowPassFilterFrequency=500
  occlusionTraceChannel=ECC_GameTraceChannel11
  stereoSpread=1000
  attenuationShapeExtents {x=0}
 }
}
